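# Unlit glitch screen material for the Wasteland battle stage, with per-screen variation through custom primitive data.
# An animated flipbook, LCD pixel pattern and random blocks combine with hue-shiftable colors that switch on the audio-reactive measure timing.
# An edge mask, optional crack overlay and reflection-mapped crystal tint shape the emissive, with intensity switches for Lumen, low quality and mobile.
# Scanline nodes are present but appear disconnected from that emissive chain.
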Material /SparksMaps/Environment/S4/Battlestage/Sets/Wasteland_Screens/Materials/M_BattleStage_Screens_Glitch_1
graph {
 properties {
  bAutomaticallySetUsageInEditor=false
  bUsedWithNanite=true
  pixelDepthOffsetMode=PDOM_Legacy
  shadingModel=MSM_Unlit
  usedShadingModels=Unlit
 }
 nodes [
  Add(A=TextureSample() {texture=/Game/Global/Textures/Default/Masks_and_Gradients/T_Grad_Rainbow}.RGB) {constB=0.4, desc=desaturate}
  n1 = Add(
   A=TextureSample(UVs=ReflectionVectorWS()) {texture=/Game/CharClasses/Technician/Gadgets/BASE/Effects/T_CrystalColor_Tile}.R
   B=VectorParameter() {parameterName=a, defaultValue=<rgba 1 0.7865 0.5 1>}.RGB
  )
  n2 = NamedRerouteUsage()
  n3 = NamedRerouteUsage()
  n4 = Multiply(A=Multiply(A=Add(A=@n2, B=@n3), B=NamedRerouteUsage()), B=NamedRerouteUsage())
  n5 = Power(
   Base=MaterialFunctionCall!(TexCoords=TextureCoordinate() {coordinateIndex=1}, Edge Sharpness=Constant() {r=3}, Edge Subtract) {materialFunction=/Game/Effects/Niagara/MasterMaterials/Functions/Mojo_MF/MF_FX_UV_EdgeMask}.1-Edge Mask
   Exp=Constant() {r=2}
  ) {constExponent=5}
  n6 = Constant() {r=0.2}
  n7 = Add(A=Multiply(A=@n4, B=Add(A=@n5, B=@n6)), B=NamedRerouteUsage()) {constA=1}
  Add(A=@n1, B=@n7)
  Constant() {desc=TestValue}
  Constant() {r=300}
  Frac(None=Time())
  MaterialFunctionCall() {materialFunction=/SparksMaps/Environment/Shared/Materials/Functions/MF_Sparks_WPO}
  MaterialFunctionCall!(CustomUPosition) {materialFunction=/SparksStageShow/Systems/StageManager/FXManager/MaterialFunctions/MF_MojoStageColors}
  n8 = ScalarParameter() {bUseCustomPrimitiveData=true, defaultValue=50, parameterName=CPD_ScanLines, primitiveDataIndex=4}
  MaterialFunctionCall!(Scanline Amount=@n8) {
   desc=MF but now using it's nodes below
   materialFunction=/Game/Environments/Asteria/Sets/NeonCity/Techart/Holograms/NewHologramMaterial/MFScanlines
  }
  MaterialFunctionCall(Emissive=NamedRerouteDeclaration(None=Multiply(A=@n1, B=@n7)) {name=Emissive, nodeColor=<rgba 0 0 0 1>}) {materialFunction=&MF_FX_EmissiveAdjustments}
  n9 = Constant() {r=0.004}
  n10 = Time()
  n11 = MaterialFunctionCall!(
   Number of Scanlines=@n8
   Scanline Speed=@n9
   "Scanlines Thickness (Bias)"
   Scanlines Dark Value=Constant() {r=0.5}
   Scanlines Light Value
   Block Opacity=Constant() {r=0.2}
   UVs=TextureCoordinate()
   Time=Multiply(A=Multiply(A=@n9, B=@n10)) {constB=20}
   Block Repeat=Constant() {r=3}
  ) {materialFunction=/Game/Effects/Fort_Effects/Effects/Atlas/Materials/MF_Atlas_Scanlines_01}
  n12 = Multiply(
   A=@n11.Secondary Lines
   B=ScalarParameter() {bUseCustomPrimitiveData=true, defaultValue=50, parameterName=CPD_AccentScanline, primitiveDataIndex=5}
  ) {constB=3}
  Multiply(A=@n4, B=LinearInterpolate(Alpha=Add(A=@n11.Result, B=@n12)) {constA=0.5})
  n13 = ScalarParameter() {defaultValue=100, parameterName=LCD_Size}
  n14 = TextureCoordinate()
  NamedRerouteDeclaration(None=TextureSample(UVs=Multiply(A=@n13, B=@n14)) {samplerType=SAMPLERTYPE_Masks, texture=/Game/Environments/Asteria/Sets/NeonCity/Techart/Textures/T_LCD_Panels_LCDSoft}.RGB) {name=LCD pattern, nodeColor=<rgba 0.5703 1 0 1>}
  n15 = Time()
  n16 = ScalarParameter() {defaultValue=0.5, parameterName=Speed}
  n17 = LinearInterpolate(
   A=Constant() {r=-3}
   Alpha=MaterialFunctionCall(Value=Multiply(A=@n15, B=@n16)) {materialFunction=&LinearSine}.Linear Sine
  ) {constA=-10, constB=5}
  n18 = Multiply(
   A=MaterialFunctionCall!(
    Texture=TextureObjectParameter() {parameterName=Screen Details Flipbook, texture=/Game/Athena/Environments/Textures/T_NeonCat_MonitorB}
    UVs=TextureCoordinate()
    SubImages=Constant2Vector() {g=3, r=1}
    Frame=Clamp(None=@n17) {maxDefault=24}
   ) {materialFunction=/Engine/Functions/Engine_MaterialFunctions03/Texturing/SubUV_Function}.RGB
   B=ScalarParameter() {bUseCustomPrimitiveData=true, defaultValue=0.5, parameterName=CPD_FlipBookViz, primitiveDataIndex=3}
  ) {constB=0.5, desc=darken}
  NamedRerouteDeclaration(None=Clamp(None=Add(A=@n18) {constB=0.1, desc=desaturate})) {name=LCD RGB, nodeColor=<rgba 1 0.7734 0 1>}
  n19 = VectorParameter() {parameterName=Param_2, defaultValue=<rgba 0.875 0.4648 0.4881 1>}
  n20 = VectorParameter() {parameterName=Param_1, defaultValue=<rgba 0.1839 0.5827 0.5885 1>}
  n21 = ScalarParameter() {bUseCustomPrimitiveData=true, defaultValue=1, parameterName=CPD_Color}
  n22 = VectorParameter() {parameterName=Param_3, defaultValue=<rgba 0.2563 0.5 0.7812 1>}
  n23 = VectorParameter() {defaultValue=<rgba 0.8255 0.4219 1 1>}
  n24 = ScalarParameter() {bUseCustomPrimitiveData=true, defaultValue=1, parameterName=CPD_Color}
  n25 = Multiply(A=CollectionParameter() {collection=/FMAudioReactiveData/AudioAnalysis/MPC_Sparks_Audio_Parameters, parameterName=MeasureLifetime}) {constB=2}
  n26 = MaterialFunctionCall(
   Hue Shift Percentage=ScalarParameter() {bUseCustomPrimitiveData=true, parameterName=CPD_HueSHift, primitiveDataIndex=6}
   Texture=LinearInterpolate(A=LinearInterpolate(A=@n19.RGB, B=@n20.RGB, Alpha=@n21), B=LinearInterpolate(A=@n22.RGB, B=@n23.RGB, Alpha=@n24), Alpha=Floor(None=@n25))
  ) {materialFunction=&HueShift}
  n27 = ScalarParameter() {defaultValue=0.2, parameterName=Desaturate}
  NamedRerouteDeclaration(None=Multiply(A=Desaturation(None=@n26.Result, Fraction=@n27), B=ScalarParameter() {defaultValue=0.7, parameterName=Darken})) {name=Color, nodeColor=<rgba 0 1 0.5078 1>}
  n28 = Multiply(A=Time(), B=ScalarParameter() {defaultValue=0.2, parameterName=BlockSpeed})
  n29 = Multiply(
   A=TextureCoordinate()
   B=AppendVector(
    A=ScalarParameter() {bUseCustomPrimitiveData=true, defaultValue=10, parameterName=SquareSize_X, primitiveDataIndex=1}
    B=ScalarParameter() {bUseCustomPrimitiveData=true, defaultValue=10, parameterName=SquareSize_Y, primitiveDataIndex=2}
   )
  )
  NamedRerouteDeclaration(None=LinearInterpolate(Alpha=MaterialFunctionCall!(PrimaryTextureCoordinates=@n29, 2ndOrderTextureCoordinates=@n29, RandomSeed=@n28, NoiseDensityBias, Bypass Noise Influence Crosstalk) {materialFunction=/Game/Effects/Fort_Effects/Materials/_Functions/MF_RandomUVSample}.NoiseWithRange) {constA=0.5}) {name=Random Square or Hex Patterns, desc=Swapped squares to Hex, nodeColor=<rgba 0 0.1328 1 1>}
  n30 = Multiply(A=Constant() {r=2}, B=NamedRerouteUsage())
  n31 = Constant()
  n32 = LinearInterpolate(
   A=Constant() {r=-0.25}
   B=Constant()
   Alpha=ScalarParameter() {bUseCustomPrimitiveData=true, parameterName=CPD_CrackCenter, primitiveDataIndex=8}
  ) {constAlpha=1}
  n33 = TextureSample(UVs=Add(A=TextureCoordinate(), B=AppendVector(A=@n31, B=@n32))) {texture=/Game/Effects/Fort_Effects/Textures/Tile/T_Cracked_Ice_D}
  NamedRerouteDeclaration(None=LinearInterpolate(B=Multiply(A=@n30, B=@n33.RGB), Alpha=ScalarParameter() {bUseCustomPrimitiveData=true, parameterName=CPD_Cracked?, primitiveDataIndex=7}) {constAlpha=1}) {name=Crack, nodeColor=<rgba 1 0.0469 0 1>}
  ScalarParameter() {defaultValue=10, parameterName=SquareSize}
  n34 = NamedRerouteUsage()
  n35 = Multiply(A=@n34, B=ScalarParameter() {defaultValue=3, parameterName=EmissiveIntensity})
  n36 = Multiply(A=@n35, B=ScalarParameter() {defaultValue=140, group=Emissive Adjustments, parameterName=LumenBoost})
  n37 = Multiply(A=@n34, B=ScalarParameter() {defaultValue=10, parameterName=EmissiveIntensityLowEnd}) {constB=1.5, desc=Low End and Mobile Screen Multiplier}
  ShadingPathSwitch(Default=QualitySwitch(Default=RayTracingQualitySwitch(Normal=@n35, RayTraced=@n36), Low=@n37), Mobile=@n37)
 ]
}
