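# Unlit glitch-screen material for the Wasteland Screens set of the Sparks Battlestage.
# A pixelated radial gradient, LCD pattern and random block reroutes form the screen layers, with a slight rainbow-gradient tint picked by custom primitive data.
# An edge-mask vignette and scanlines shape the result before emissive adjustments.
# An additive glitch layer multiplies panning digital noise by rectangular masks that noise textures shift over time.
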Material /SparksMaps/Environment/S4/Battlestage/Sets/Wasteland_Screens/Materials/M_BattleStage_Screens_Glitch_2
graph {
 properties {bAutomaticallySetUsageInEditor=false, pixelDepthOffsetMode=PDOM_Legacy, shadingModel=MSM_Unlit, usedShadingModels=Unlit}
 nodes [
  Constant3Vector() {constant=<rgba 0.9277 0.95 1 1>}
  Constant() {desc=TestValue, r=0.3}
  MaterialFunctionCall() {materialFunction=/SparksMaps/Environment/Shared/Materials/Functions/MF_Sparks_WPO}
  n1 = Multiply(A=Multiply(A=NamedRerouteUsage(), B=NamedRerouteUsage()), B=NamedRerouteUsage())
  n2 = NamedRerouteDeclaration(None=TextureSample(UVs=ScalarParameter() {bUseCustomPrimitiveData=true, defaultValue=1, parameterName=CPD_Color}) {texture=/Game/Global/Textures/Default/Masks_and_Gradients/T_Grad_Rainbow}.RGB) {name=CPDColor, nodeColor=<rgba 1 0 0.9609 1>}
  n3 = ScalarParameter() {defaultValue=0.004, parameterName=WashIntensity}
  n4 = MaterialFunctionCall!(TexCoords, Edge Sharpness=Constant() {r=5}, Edge Subtract) {materialFunction=/Game/Effects/Niagara/MasterMaterials/Functions/Mojo_MF/MF_FX_UV_EdgeMask}
  n5 = ScalarParameter() {defaultValue=5, parameterName=Param_1}
  n6 = Multiply(
   A=Multiply(A=LinearInterpolate(A=@n1, B=@n2, Alpha=@n3), B=ScalarParameter() {defaultValue=20, parameterName=Brightness})
   B=Multiply(A=Power(Base=@n4.1-Edge Mask, Exp=@n5) {constExponent=5}) {constB=1.15}
  ) {desc=Vignette}
  n7 = LinearInterpolate(Alpha=MaterialFunctionCall!(Scanline Amount=Constant() {r=500}) {materialFunction=/Game/Environments/Asteria/Sets/NeonCity/Techart/Holograms/NewHologramMaterial/MFScanlines}.Result) {constA=0.5}
  MaterialFunctionCall(Emissive=Add(A=NamedRerouteUsage(), B=Multiply(A=@n6, B=@n7) {desc=Scanlines})) {materialFunction=&MF_FX_EmissiveAdjustments}
  n8 = TextureCoordinate() {uTiling=0.5}
  n9 = ScalarParameter() {defaultValue=50, parameterName=SquareSize}
  NamedRerouteDeclaration(None=TextureSample(UVs=Multiply(A=@n9, B=@n8)) {samplerType=SAMPLERTYPE_Masks, texture=/Game/Environments/Asteria/Sets/NeonCity/Techart/Textures/T_LCD_Panels_LCDSoft}.RGB) {name=LCD pattern, nodeColor=<rgba 0.5703 1 0 1>}
  n10 = VectorParameter() {parameterName=Color Light, defaultValue=<rgba 0.0027 0.2789 0.7305 1>}
  n11 = Multiply(A=TextureCoordinate(), B=@n9)
  n12 = Round(None=@n11)
  n13 = Multiply(A=MaterialFunctionCall(UVs=Divide(A=@n12, B=@n9), Radius=Constant() {r=0.5}, Density=Constant() {r=0.5}) {materialFunction=&RadialGradientExponential}.RadialGradientExponential) {constB=1.5, desc=Boost Brightness of Radial Grad}
  n14 = LinearInterpolate(
   A=LinearInterpolate(A=VectorParameter() {parameterName=Color Dark, defaultValue=<rgba 0.0021 0.013 0.1198 1>}.RGB, B=Add(A=@n10.RGB), Alpha=Saturate(None=@n13))
   B=VectorParameter() {parameterName=Frame Color, defaultValue=<rgba 0.0009 0.5395 1 1>}.RGB
  ) {constAlpha=0}
  n15 = LinearInterpolate()
  NamedRerouteDeclaration(None=Multiply(A=Multiply(A=@n14, B=@n15) {desc=darken outside the main window frame}) {constB=0.5, desc=Overall Darken}) {name=PrimaryScreenColor, nodeColor=<rgba 1 0 0.8672 1>}
  n16 = Multiply(A=Time(), B=ScalarParameter() {defaultValue=0.5, parameterName=BlockSpeed})
  n17 = Multiply(A=@n11) {constB=0.1}
  NamedRerouteDeclaration(None=LinearInterpolate(Alpha=MaterialFunctionCall!(PrimaryTextureCoordinates=@n17, 2ndOrderTextureCoordinates=@n17, RandomSeed=@n16, NoiseDensityBias, Bypass Noise Influence Crosstalk) {materialFunction=/Game/Effects/Fort_Effects/Materials/_Functions/MF_RandomUVSample}.NoiseWithRange)) {name=Random Square or Hex Patterns, desc=Swapped squares to Hex, nodeColor=<rgba 0 0.1328 1 1>}
  n18 = TextureCoordinate()
  n19 = Constant3Vector() {constant=<rgba 5 0.05 0 1>}
  n20 = TextureSample(UVs=Multiply(A=Panner(Coordinate=@n18) {speedX=0.1, speedY=1}, B=ComponentMask(None=@n19) {g=true, r=true})) {samplerType=SAMPLERTYPE_LinearColor, texture=/Game/Global/Textures/Default/Noise/T_Atlas_DigitalNoise_01}
  n21 = Multiply(
   A=Panner(Coordinate=TextureCoordinate()) {speedX=-0.2, speedY=0.2}
   B=ComponentMask(None=Constant3Vector() {constant=<rgba 4 0.02 0 1>}) {g=true, r=true}
  )
  n22 = Multiply(A=Time()) {constB=0.4}
  n23 = TextureSample(UVs=@n22) {texture=&T_NoisePack_01}
  n24 = LinearInterpolate(Alpha=OneMinus(None=@n23.B)) {constA=1}
  n25 = TextureCoordinate()
  n26 = LinearInterpolate(Alpha=OneMinus(None=@n23.R)) {constA=-1}
  n27 = LinearInterpolate(A=@n24, B=Subtract(B=@n24) {constB=0}, Alpha=ComponentMask(None=Add(A=@n26, B=@n25)) {r=true})
  n28 = Constant() {r=0.75}
  n29 = Constant()
  n30 = If(A=@n27, B=Constant() {r=0.25}, "A > B"=Constant() {r=1}, "A < B"=Constant())
  n31 = OneMinus(None=TextureSample(UVs=@n22) {samplerType=SAMPLERTYPE_Normal, texture=/Game/Global/Textures/Default/Noise/64x64Noise}.G)
  n32 = ComponentMask(None=Add(A=@n25, B=LinearInterpolate(Alpha=@n31)) {constB=0}) {g=true}
  n33 = Constant() {r=0.75}
  n34 = Constant()
  n35 = If(A=@n32, B=Constant() {r=0.25}, "A > B"=Constant() {r=1}, "A < B"=Constant())
  n36 = Multiply(
   A=Multiply(
    A=@n20.B
    B=TextureSample(UVs=@n21) {samplerType=SAMPLERTYPE_LinearColor, texture=&T_Atlas_DigitalNoise_01}.G
   )
   B=Multiply(A=If(A=@n27, B=@n28, "A > B"=@n29, "A < B"=@n30), B=If(A=@n32, B=@n33, "A > B"=@n34, "A < B"=@n35))
  )
  n37 = Multiply(A=Add(A=NamedRerouteUsage()) {constB=0.3}) {constB=0.5}
  n38 = LinearInterpolate(A=@n37, B=LinearInterpolate(B=@n37, Alpha=@n20.R), Alpha=@n20.R)
  NamedRerouteDeclaration(None=Multiply(A=Multiply(A=@n38, B=@n36)) {constB=0.25}) {name=Glitch, nodeColor=<rgba 1 0 0.2812 1>}
  VectorParameter() {parameterName=WashColor, defaultValue=<rgba 0.2893 0.47 0.9734 1>}
 ]
}
